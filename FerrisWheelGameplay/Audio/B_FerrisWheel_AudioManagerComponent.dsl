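; Audio manager component for the Ferris wheel zipline loop that plays music while the locally viewed player rides it.
; On BeginPlay it casts its owner to BP_FerrisWheel_Zipline_Loop and binds to the begin and end ziplining events.
; Begin spawns and fades in 2D music if none is valid, registering it with the MusicEventSubsystem under an event tag.
; End fades that music out and removes its event from the subsystem.
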(event EventBeginPlay
  (|CallParentFunction)
  (bind _asbp_ferris_wheel_zipline_loop (Utilities|Casting|CastToBP_FerrisWheel_Zipline_Loop (Components|GetOwner)))
  (Default|AssignOnPlayerBeginZiplining _asbp_ferris_wheel_zipline_loop (delegate Custom|OnPlayerBeginZiplining_Event))
  (Default|AssignOnPlayerEndZiplining _asbp_ferris_wheel_zipline_loop (delegate Custom|OnPlayerEndZiplining_Event)))

(event Custom|OnPlayerEndZiplining_Event
  (params (PawnEndingZipline :type Fort%20Player%20Pawn%20Object%20Reference))
  (if (Pawn|IsLocallyViewed PawnEndingZipline)
    (bind _ferriswheelmusic (Variables|Default|GetFerrisWheelMusic)
      (:then
        (Audio|Components|Audio|FadeOut _ferriswheelmusic (Variables|Default|GetFade%20Out%20Duration) 0.0 "Linear" :!class /Script/Engine.AudioComponent)
        (Audio|RemoveEvent (GameInstanceSubsystems|GetMusicEventSubsystem) _ferriswheelmusic))
      (:else))))

(event Custom|OnPlayerBeginZiplining_Event
  (params (PawnZiplining :type Fort%20Player%20Pawn%20Object%20Reference))
  (if (Pawn|IsLocallyViewed PawnZiplining)
    (Variables|Default|GetFerrisWheelMusic
      (:then)
      (:else
        (bind _musictoplay (Variables|Default|GetMusicToPlay)
          (:then
            (bind _returnvalue (Audio|SpawnSound2D _musictoplay))
            (bind _output_get (Variables|Default|SetFerrisWheelMusic _returnvalue))
            (Audio|Components|Audio|FadeIn _returnvalue 2.0 1.0 0.0 "Linear" :!class /Script/Engine.AudioComponent)
            (Audio|AddEventbyComponent (GameInstanceSubsystems|GetMusicEventSubsystem) _output_get (Variables|Default|GetEvent%20Tag) "(GameplayTags=)" 0 "MusicPartyWonkeeland"))
          (:else))))))
