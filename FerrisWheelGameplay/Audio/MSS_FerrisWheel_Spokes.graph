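# One-shot mono MetaSound Source for ferris wheel spoke ambience, spatialized with its own attenuation and routed to an ambient one-shot sound class.
# On play, a random get picks one of seven spoke waves and a mono Wave Player plays it, firing On Finished when done.
# The Period input is declared with a default but is not connected in the graph, so any repetition is not handled here.
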MetaSoundSource /FerrisWheelGameplay/Audio/MSS_FerrisWheel_Spokes
inputs [
 In Array {type=WaveAsset:Array}
 Period {type=Time}
 UE.Source.OnPlay {type=Trigger}
]
interfaces [UE.OutputFormat.Mono:1.0, UE.Source.OneShot:1.0, UE.Source:1.1]
outputs [
 "UE.OutputFormat.Mono.Audio:0" {type=Audio}
 UE.Source.OneShot.OnFinished {type=Trigger}
]
pages [Default {
  input_defaults {
   Period=0.588
   UE.Source.OnPlay=null
   In Array [
    /FerrisWheelGameplay/Audio/Soundwaves/FerrisWheel_Amb_Spokes_01
    /FerrisWheelGameplay/Audio/Soundwaves/FerrisWheel_Amb_Spokes_02
    /FerrisWheelGameplay/Audio/Soundwaves/FerrisWheel_Amb_Spokes_03
    /FerrisWheelGameplay/Audio/Soundwaves/FerrisWheel_Amb_Spokes_04
    /FerrisWheelGameplay/Audio/Soundwaves/FerrisWheel_Amb_Spokes_05
    /FerrisWheelGameplay/Audio/Soundwaves/FerrisWheel_Amb_Spokes_06
    /FerrisWheelGameplay/Audio/Soundwaves/FerrisWheel_Amb_Spokes_07
   ]
  }
  nodes [
   ArrayRandomGet_0 = Array.Random Get:WaveAsset:Array(In Array=^In Array, Next=^UE.Source.OnPlay)
   WavePlayer_0 = UE.Wave Player:Mono(Play=^UE.Source.OnPlay, Wave Asset=@ArrayRandomGet_0.Value) {also {version [1,1]}}
  ]
  links [{dst="@UE.Source.OneShot.OnFinished", src=WavePlayer_0.On Finished}, {dst="@UE.OutputFormat.Mono.Audio:0", src=WavePlayer_0.Out Mono}]
 }]
properties {
 assetImportData=&self:AssetImportData
 attenuationSettings=&/Amb_FerrisWheel_Spokes_Att
 maxDistance=2250
 sampleRate=48000
 soundClassObject=&Fort_Ambient_OneShots
 volume=4
}
