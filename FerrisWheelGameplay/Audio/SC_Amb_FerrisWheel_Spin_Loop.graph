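# Looping ambient SoundCue for the spinning Ferris wheel, routed to the Athena ambient sound class.
# A distance crossfade blends a close layer into a distant layer, each mixing a wheel ambience loop with a motor loop; its third input is empty.
# Each layer overrides attenuation with its own falloff, listener focus, occlusion and reverb send settings.
# Only the distant layer adds distance-based low-pass filtering.
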SoundCue /FerrisWheelGameplay/Audio/SC_Amb_FerrisWheel_Spin_Loop
root=DistanceCrossFade_0
nodes [
 {
  id=Attenuation_0
  children [Mixer_1]
  properties {
   bOverrideAttenuation=true
   attenuationOverrides {
    bAttenuateWithLPF=true
    bEnableFocusInterpolation=true
    bEnableListenerFocus=true
    falloffDistance=8000
    focusAttackInterpSpeed=0.5
    focusAzimuth=45
    focusReleaseInterpSpeed=0.5
    lPFFrequencyAtMin=15309.9775
    lPFRadiusMax=8000
    nonFocusAzimuth=90
    nonFocusVolumeAttenuation=0.3
    occlusionLowPassFilterFrequency=10000
    occlusionVolumeAttenuation=0.2701
    reverbDistanceMax=5000
    reverbDistanceMin=1000
    reverbWetLevelMax=1
    reverbWetLevelMin=0.5
    stereoSpread=3000
    attenuationShapeExtents=<1000 2000>
   }
  }
 }
 {
  id=Attenuation_4
  children [Mixer_0]
  properties {
   bOverrideAttenuation=true
   attenuationOverrides {
    bEnableFocusInterpolation=true
    bEnableListenerFocus=true
    falloffDistance=4500
    focusAttackInterpSpeed=0.5
    focusAzimuth=45
    focusReleaseInterpSpeed=0.5
    nonFocusAzimuth=120
    nonFocusDistanceScale=0.5
    nonFocusVolumeAttenuation=0.5
    occlusionLowPassFilterFrequency=10000
    occlusionVolumeAttenuation=0.2701
    reverbDistanceMax=1500
    reverbDistanceMin=500
    reverbWetLevelMax=1
    reverbWetLevelMin=0.5
    stereoSpread=6000
    attenuationShapeExtents=<1000 2000>
   }
  }
 }
 {
  id=DistanceCrossFade_0
  children [Attenuation_4, Attenuation_0, null]
  properties {crossFadeInput [
    {fadeInDistanceEnd=250, fadeInDistanceStart=0, fadeOutDistanceEnd=6000, fadeOutDistanceStart=3000, volume=1}
    {fadeInDistanceEnd=5000, fadeInDistanceStart=2500, fadeOutDistanceEnd=10000, fadeOutDistanceStart=5000, volume=1}
    {fadeInDistanceEnd=5000, fadeInDistanceStart=2500, fadeOutDistanceEnd=10000, fadeOutDistanceStart=5000, volume=1}
   ]}
 }
 {id=Mixer_0, children [WavePlayer_0,WavePlayer_2], properties {inputVolume [2.75,2.5]}}
 {id=Mixer_1, children [WavePlayer_1,WavePlayer_3], properties {inputVolume [4.5,1.75]}}
 {id=WavePlayer_0, properties {bLooping=true, soundWaveAssetPtr=/FerrisWheelGameplay/Audio/Soundwaves/FerrisWheel_Amb_Close_Loop}}
 {id=WavePlayer_1, properties {bLooping=true, soundWaveAssetPtr=/FerrisWheelGameplay/Audio/Soundwaves/FerrisWheel_Amb_Distant_Loop}}
 {id=WavePlayer_2, properties {bLooping=true, soundWaveAssetPtr=/FerrisWheelGameplay/Audio/Soundwaves/FerrisWheel_Amb_Motor_Close_Loop}}
 {id=WavePlayer_3, properties {bLooping=true, soundWaveAssetPtr=/FerrisWheelGameplay/Audio/Soundwaves/FerrisWheel_Amb_Motor_Distant_Loop}}
]
properties {
 maxDistance=10000
 soundClassObject=&Fort_Ambient_Athena
 volumeMultiplier=1.75
 attenuationOverrides {bSpatialize=false, falloffDistance=2000, stereoSpread=1147.9967, attenuationShapeExtents {x=6000}}
}
